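# Unlit translucent material for Niagara mesh particles that renders snot on the DoubleDuty backpack.
# Emissive blends highlight and base colors by Fresnel, tinted by particle color, plus two UV sphere-mask glints.
# Particle alpha drives a randomly offset Voronoi dissolve, replaced by a constant on mobile, which masks the Fresnel-and-glint opacity.
# That same mask scales a vertex-normal offset that grows as particle alpha falls.
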Material /BRCosmetics/Accessories/FORT_Backpacks/Backpack_M_MED_DoubleDuty/FX/M_DoubleDuty_Backpack_Snot
graph {
 properties {
  bUsedWithNiagaraMeshParticles=true
  blendMode=BLEND_Translucent
  pixelDepthOffsetMode=PDOM_Legacy
  shadingModel=MSM_Unlit
  usedShadingModels=Unlit
 }
 nodes [
  n1 = ParticleColor()
  n2 = LinearInterpolate(
   A=VectorParameter() {parameterName=ColorHighlight, defaultValue=<rgba 0.32 0.4 0.3533 1>}.RGB
   B=VectorParameter() {parameterName=ColorBase, defaultValue=<rgba 0.1005 0.15 0.117 1>}.RGB
   Alpha=OneMinus(None=Fresnel() {exponent=3.5})
  )
  n3 = MaterialFunctionCall(Float2=TextureCoordinate()) {materialFunction=&BreakOutFloat2Components}
  n4 = Multiply(A=@n3.R) {constB=2}
  n5 = Add(A=@n4) {constB=-0.75}
  n6 = Add(A=@n3.G) {constB=0.1}
  n7 = Add(A=@n4) {constB=-0.85}
  n8 = Add(A=@n3.G) {constB=0}
  n9 = Add(
   A=SphereMask(A=AppendVector(A=@n5, B=@n6), B=Constant() {r=0.5}) {attenuationRadius=0.06, hardnessPercent=60}
   B=SphereMask(A=AppendVector(A=@n7, B=@n8), B=Constant() {r=0.5}) {attenuationRadius=0.04, hardnessPercent=30}
  )
  MaterialFunctionCall(Emissive=Add(A=Multiply(A=@n1.RGB, B=@n2), B=@n9)) {materialFunction=&MF_FX_EmissiveAdjustments}
  n10 = ParticleColor()
  n11 = MaterialFunctionCall(Float2=Add(A=ParticleRandom(), B=TextureCoordinate())) {materialFunction=&BreakOutFloat2Components}
  n12 = Multiply(A=@n11.R)
  n13 = Multiply(A=@n11.G) {constB=0.6}
  n14 = TextureSampleParameter2D(UVs=Multiply(A=AppendVector(A=@n12, B=@n13)) {constB=0.5}) {parameterName=Dissolve, texture=/BRCosmetics/Athena/Effects/Particles/Skydiving/NS_Contrail_PinkSpike/Textures/T_PinkSpike_Voronoi2D_Height}
  n15 = SmoothStep(Value=@n14.R) {constMax=0.7, constMin=0.3}
  n16 = Constant() {r=0.5}
  n17 = Multiply(A=@n10.A) {constB=2}
  n18 = Saturate(None=Add(A=ShadingPathSwitch(Default=@n15, Deferred=@n15, Mobile=@n16) {desc=Migrated from FeatureLevelSwitch}, B=Subtract(A=@n17)))
  n19 = VertexNormalWS()
  Multiply(A=OneMinus(None=@n10.A), B=Multiply(A=Multiply(A=@n18, B=@n19)) {constB=7})
  n20 = Fresnel() {baseReflectFraction=0, exponent=0.2}
  Multiply(A=Add(A=@n9, B=LinearInterpolate(Alpha=@n20) {constA=0.2}), B=@n18)
 ]
}
