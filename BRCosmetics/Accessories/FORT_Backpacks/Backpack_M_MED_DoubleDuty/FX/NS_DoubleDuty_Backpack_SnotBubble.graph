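# Niagara effect for the DoubleDuty backpack cosmetic: a snot bubble that swells, wobbles and pops.
# Snot_Bubble burst-spawns a mesh following the backpack skeletal mesh; one BubbleAnimation curve drives both its scale and X offset, so it grows outward from its attach point.
# Its death event pairs with Snot_Burst, a short-lived green SubUV sprite with a fading dynamic material parameter.
# Both emitters tint their color by time of day.
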NiagaraSystem /BRCosmetics/Accessories/FORT_Backpacks/Backpack_M_MED_DoubleDuty/FX/NS_DoubleDuty_Backpack_SnotBubble
dependencies {
 usedDataInterfaces [
  &NiagaraDataInterfaceCurve
  &NiagaraDataInterfaceSpriteRendererInfo
  &NiagaraDataInterfaceSkeletalMesh
 ]
 usedDynamicInputs [&RandomRangeVector2D]
 usedModules [
  &SystemState
  &self:SystemSpawnScript.SystemScriptSource.SystemScriptGraph.NiagaraNodeAssignment_0.SetVariables_#1
 ]
}
emitters {
 Snot_Bubble {
  emitterUpdateScript [
   &EmitterState()
   &SpawnBurst_Instantaneous(Use Spawn Probability=false, UseLoopCountLimit=false)
  ]
  particleSpawnScript [
   &InitializeParticle(
    Color Mode=::NewEnumerator0
    Lifetime=6.1
    Mesh Scale Mode=::NewEnumerator3
    Mesh Uniform Scale=0.0255
    UsePositionOffset=false
    Write Material Random=true
   )
   &CameraOffset(Camera Offset Amount=3) {enabled=false}
   &InitialMeshOrientation(Mesh Orientation Mode=::NewEnumerator2, Use Rotation Vector=false)
   /SetVariables(Particles.Color=&FNTimeOfDayColor(
     Night Color=<rgba 0.7237 0.72 0.8 1>
     Morning Color=<rgba 0.8 0.72 0.7491 1>
     Day Color=<rgba 1 1 1 1>
     Evening Color=<rgba 0.8 0.7389 0.704 1>
    )) {isSetParameters=true}
  ]
  particleUpdateScript [
   /SetVariables(Particles.BubbleAnimation:NiagaraFloat=&FloatFromCurve(FloatCurve=DataInterface() {propertyValues {Curve {keys [
         {interpMode=RCIM_Cubic, tangentMode=RCTM_User, time=0, value=0}
         {arriveTangent=0.043, interpMode=RCIM_Cubic, leaveTangent=0.043, tangentMode=RCTM_User, time=0.15, value=0.7}
         {arriveTangent=0.5556, interpMode=RCIM_Cubic, leaveTangent=0.5556, time=0.3, value=0.4}
         <at 0.42 0.85 RCIM_Cubic>
         {arriveTangent=-1.087, interpMode=RCIM_Cubic, leaveTangent=-1.087, time=0.56, value=0.4}
         {arriveTangent=-0.2273, interpMode=RCIM_Cubic, leaveTangent=-0.2273, time=0.65, value=0.6}
         {arriveTangent=1.1429, interpMode=RCIM_Cubic, leaveTangent=1.1429, time=0.78, value=0.35}
         {arriveTangent=0.9319, interpMode=RCIM_Cubic, leaveTangent=0.9319, tangentMode=RCTM_User, time=1, value=1}
        ]}}})) {isSetParameters=true}
   &SkeletalMeshLocation(
    Bone Sampling Mode=::NewEnumerator4
    MaskSpawnGroup=false
    Orientation Sampling=::NewEnumerator1
    Sampled Position Offset=&MakeVector(X=&Lerp_Float(A=-4.7, B=-7.5, Alpha=$Particles.BubbleAnimation), Y=0, Z=0)
    Skeletal Mesh=$System.Skeletal Mesh
   )
   &ParticleState()
   &ScaleMeshSize(Scale Factor=&VectorFromFloat(Value=$Particles.BubbleAnimation))
   &SolveForcesAndVelocity(Clamp Velocity=false, Limit Acceleration=false)
   &GenerateDeathEvent(Use Event Probability=false)
   &ScaleColor(
    Scale Alpha=&FloatFromCurve(FloatCurve=DataInterface() {propertyValues {Curve {keys [<at 0.99 1>,<at 1 0>]}}})
    ScaleA=true
    ScaleRGB=false
   )
  ]
  rendererClasses [&NiagaraMeshRendererProperties]
  renderers [{rendererClass=&NiagaraMeshRendererProperties, rendererIndex=0}]
 }
 Snot_Burst {
  emitterUpdateScript [&EmitterState()]
  particleSpawnScript [
   &InitializeParticle(
    Color=<rgba 0.42 0.7 0.56 1.6>
    Lifetime=0.18
    Sprite Rotation Mode=::NewEnumerator2
    Sprite Size Mode=::NewEnumerator3
    Uniform Sprite Size=16
    UsePositionOffset=false
    Write Material Random=true
   )
   &CameraOffset(Camera Offset Amount=5)
   &AddVelocity(Velocity=<0 0 5>)
   &SubUVAnimation(
    "Sprite Renderer "=DataInterface() {propertyValues {SpriteRenderer=&self:Snot_Burst_0.NiagaraSpriteRendererProperties_0}}
    SubUV Animation Mode:ENiagara_SubUVLookupModeV2=::NewEnumerator2
    UseEndFrame=false
    UseStartFrame=false
   )
   &InheritVelocity(Speed Limit=true, UseSpeedThreshold=false)
   /SetVariables(Particles.Color=&FNTimeOfDayColor(
     Night Color=<rgba 0.25 0.5 0.375 0.5>
     Morning Color=<rgba 0.225 0.45 0.3375 1>
     Day Color=<rgba 0.24 0.6 0.42 1>
     Evening Color=<rgba 0.2 0.5 0.35 0.7>
    )) {isSetParameters=true}
  ]
  particleUpdateScript [
   &ParticleState()
   &SpriteRotationRate(Rotation Rate=10)
   &SolveForcesAndVelocity(Clamp Velocity=false, Limit Acceleration=false)
   &ScaleSpriteSize(Uniform Curve Sprite Scale=DataInterface() {propertyValues {Curve {keys [
        {arriveTangent=1.8307, interpMode=RCIM_Cubic, leaveTangent=1.8307, tangentMode=RCTM_User, time=0, value=0.7}
        {arriveTangent=0.109, interpMode=RCIM_Cubic, leaveTangent=0.109, tangentMode=RCTM_User, time=0.6, value=1}
       ]}}})
   &DynamicMaterialParameters(
    Index 0 Param 1=&FloatFromCurve(
     FloatCurve=DataInterface() {propertyValues {Curve {keys [{interpMode=RCIM_Cubic, tangentMode=RCTM_User, time=0.1, value=1}, {interpMode=RCIM_Cubic, tangentMode=RCTM_User, time=1, value=0}]}}}
     Scale Curve=&RandomRangeFloat(Minimum=0.6, Maximum=0.8)
    )
    Module.Index0Param0Write=true
    Module.Index0Param1Write=false
    Module.Index0Param2Write=false
    Module.Index0Param3Write=false
   )
  ]
  rendererClasses [&NiagaraSpriteRendererProperties]
  renderers [{rendererClass=&NiagaraSpriteRendererProperties}]
 }
}
summary {
 systemName=NS_DoubleDuty_Backpack_SnotBubble
 emitters [
  {emitterName=Snot_Bubble, rendererClasses [&NiagaraMeshRendererProperties]}
  {emitterName=Snot_Burst, rendererClasses [&NiagaraSpriteRendererProperties]}
 ]
 userVariables [
  User.bIsFrontEnd {type=NiagaraBool}
  User.bIsFrontEndPreview {type=NiagaraBool}
  User.bKillReaction {type=NiagaraBool}
  User.global Alpha {type=NiagaraFloat, defaultValue=0, description=""}
  User.IsSkydiving {type=NiagaraBool}
 ]
}
