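# Physics asset for the DoubleDuty backpack cosmetic, defining simulation bodies for its belt connections, belt strap and two bottom pack pieces.
# Each piece pairs a base_ capsule with an offset dyn_ capsule, constrained by narrow twist and per-piece swing limits, with the strap swinging widest on one axis.
# All five base_ bodies connect to base_backpack_master through constraints with free swing and twist.
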PhysicsAsset /BRCosmetics/Accessories/FORT_Backpacks/Backpack_M_MED_DoubleDuty/Meshes/M_MED_DoubleDuty_Pack_Physics
bodies {
 base_backpack_master [{shapeType=Capsule, length=6.6888, radius=1, shapeName=base_backpack_master_capsule, center=<-5.1667 -4.4436 1.9456>}]
 base_beltConnection_l [{shapeType=Capsule, length=0.505, radius=0.505, shapeName=base_beltConnection_l_capsule, rotation=<rot 0 -90 0>}]
 base_beltConnection_r [{shapeType=Capsule, length=0.505, radius=0.505, shapeName=base_beltConnection_r_capsule, rotation=<rot 0 -90 0>}]
 base_beltStrap [{shapeType=Capsule, length=0.505, radius=0.505, shapeName=base_beltStrap_capsule, rotation=<rot 0 -90 0>}]
 base_pack_bottom_01 [{shapeType=Capsule, length=0.505, radius=0.505, shapeName=base_pack_bottom_01_capsule, rotation=<rot 0 -90 0>}]
 base_pack_bottom_02 [{shapeType=Capsule, length=0.505, radius=0.505, shapeName=base_pack_bottom_02_capsule, rotation=<rot 0 -90 0>}]
 dyn_beltConnection_l [{shapeType=Capsule, length=2, radius=1, shapeName=dyn_beltConnection_l_capsule, center=<2 0 0>, rotation=<rot 90 0 0>}]
 dyn_beltConnection_r [{shapeType=Capsule, length=2, radius=1, shapeName=dyn_beltConnection_r_capsule, center=<2 0 0>, rotation=<rot 90 0 0>}]
 dyn_beltStrap [{shapeType=Capsule, length=2, radius=1, shapeName=dyn_beltStrap_capsule, center=<2 0 0>, rotation=<rot 90 0 0>}]
 dyn_pack_bottom_01 [{shapeType=Capsule, length=2, radius=1, shapeName=dyn_pack_bottom_01_capsule, center=<2 0 0>, rotation=<rot 90 0 0>}]
 dyn_pack_bottom_02 [{shapeType=Capsule, length=2, radius=1, shapeName=dyn_pack_bottom_02_capsule, center=<2 0 0>, rotation=<rot 90 0 0>}]
}
constraints [
 {bone1Name=dyn_beltConnection_l, bone2Name=base_beltConnection_l, swing1LimitDegrees=2, swing2LimitDegrees=12, twistLimitDegrees=1}
 {bone1Name=dyn_beltConnection_r, bone2Name=base_beltConnection_r, swing1LimitDegrees=2, swing2LimitDegrees=15, twistLimitDegrees=1}
 {bone1Name=dyn_pack_bottom_01, bone2Name=base_pack_bottom_01, swing1LimitDegrees=10, swing2LimitDegrees=5, twistLimitDegrees=1}
 {bone1Name=dyn_pack_bottom_02, bone2Name=base_pack_bottom_02, swing1LimitDegrees=5, swing2LimitDegrees=2.5, twistLimitDegrees=1}
 {bone1Name=dyn_beltStrap, bone2Name=base_beltStrap, swing1LimitDegrees=70, swing2LimitDegrees=5, twistLimitDegrees=1}
 {
  bone1Name=base_beltConnection_r
  bone2Name=base_backpack_master
  swing1LimitDegrees=45
  swing1Motion=Free
  swing2LimitDegrees=45
  swing2Motion=Free
  twistLimitDegrees=45
  twistMotion=Free
 }
 {
  bone1Name=base_beltConnection_l
  bone2Name=base_backpack_master
  swing1LimitDegrees=45
  swing1Motion=Free
  swing2LimitDegrees=45
  swing2Motion=Free
  twistLimitDegrees=45
  twistMotion=Free
 }
 {
  bone1Name=base_beltStrap
  bone2Name=base_backpack_master
  swing1LimitDegrees=45
  swing1Motion=Free
  swing2LimitDegrees=45
  swing2Motion=Free
  twistLimitDegrees=45
  twistMotion=Free
 }
 {
  bone1Name=base_pack_bottom_01
  bone2Name=base_backpack_master
  swing1LimitDegrees=45
  swing1Motion=Free
  swing2LimitDegrees=45
  swing2Motion=Free
  twistLimitDegrees=45
  twistMotion=Free
 }
 {
  bone1Name=base_pack_bottom_02
  bone2Name=base_backpack_master
  swing1LimitDegrees=45
  swing1Motion=Free
  swing2LimitDegrees=45
  swing2Motion=Free
  twistLimitDegrees=45
  twistMotion=Free
 }
]
