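# Physics asset for the ears of the DoubleDuty backpack cosmetic's TailEars mesh, defining bodies for jointed ear movement.
# Each side is a three-bone chain (ear_00 to ear_02) of identical small capsules laid along the bone.
# Constraints link each bone to the previous one with moderate, equal swing limits and almost no twist, so the ears bend without rolling.
# The two chains are independent, with no constraint above either ear_00 bone.
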PhysicsAsset /BRCosmetics/Accessories/FORT_Backpacks/Backpack_M_MED_DoubleDuty/Meshes/M_MED_DoubleDuty_TailEars_Physics
bodies {
 ear_00_l [{shapeType=Capsule, length=2, radius=1, shapeName=ear_00_l_capsule, center=<2 0 0>, rotation=<rot 90 0 0>}]
 ear_00_r [{shapeType=Capsule, length=2, radius=1, shapeName=ear_00_r_capsule, center=<2 0 0>, rotation=<rot 90 0 0>}]
 ear_01_l [{shapeType=Capsule, length=2, radius=1, shapeName=ear_01_l_capsule, center=<2 0 0>, rotation=<rot 90 0 0>}]
 ear_01_r [{shapeType=Capsule, length=2, radius=1, shapeName=ear_01_r_capsule, center=<2 0 0>, rotation=<rot 90 0 0>}]
 ear_02_l [{shapeType=Capsule, length=2, radius=1, shapeName=ear_02_l_capsule, center=<2 0 0>, rotation=<rot 90 0 0>}]
 ear_02_r [{shapeType=Capsule, length=2, radius=1, shapeName=ear_02_r_capsule, center=<2 0 0>, rotation=<rot 90 0 0>}]
}
constraints [
 {bone1Name=ear_01_l, bone2Name=ear_00_l, swing1LimitDegrees=25, swing2LimitDegrees=25, twistLimitDegrees=1}
 {bone1Name=ear_02_l, bone2Name=ear_01_l, swing1LimitDegrees=25, swing2LimitDegrees=25, twistLimitDegrees=1}
 {bone1Name=ear_01_r, bone2Name=ear_00_r, swing1LimitDegrees=25, swing2LimitDegrees=25, twistLimitDegrees=1}
 {bone1Name=ear_02_r, bone2Name=ear_01_r, swing1LimitDegrees=25, swing2LimitDegrees=25, twistLimitDegrees=1}
]
