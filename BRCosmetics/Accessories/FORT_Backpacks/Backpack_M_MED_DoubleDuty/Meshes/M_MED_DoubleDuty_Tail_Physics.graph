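# Physics asset for the tail of the DoubleDuty backpack cosmetic, defining its collision bodies and joints.
# One offset, rotated capsule covers the backpack master bone, and eight small identical capsules follow the tail bones in a chain.
# Swing limits widen from tail root toward tip, leveling off at the last joints, while twist stays tight.
# The first tail bone's joint to the backpack leaves swing and twist free.
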PhysicsAsset /BRCosmetics/Accessories/FORT_Backpacks/Backpack_M_MED_DoubleDuty/Meshes/M_MED_DoubleDuty_Tail_Physics
bodies {
 base_backpack_master [{
   shapeType=Capsule
   length=13.6946
   radius=5.961
   shapeName=base_backpack_master_capsule
   center=<5.773 -9.4437 3.0541>
   rotation=<rot 12.0933 0.0001 0>
  }]
 tail_01 [{shapeType=Capsule, length=2, radius=1, shapeName=tail_01_capsule}]
 tail_02 [{shapeType=Capsule, length=2, radius=1, shapeName=tail_02_capsule}]
 tail_03 [{shapeType=Capsule, length=2, radius=1, shapeName=tail_03_capsule}]
 tail_04 [{shapeType=Capsule, length=2, radius=1, shapeName=tail_04_capsule}]
 tail_05 [{shapeType=Capsule, length=2, radius=1, shapeName=tail_05_capsule}]
 tail_06 [{shapeType=Capsule, length=2, radius=1, shapeName=tail_06_capsule}]
 tail_07 [{shapeType=Capsule, length=2, radius=1, shapeName=tail_07_capsule}]
 tail_08 [{shapeType=Capsule, length=2, radius=1, shapeName=tail_08_capsule}]
}
constraints [
 {bone1Name=tail_02, bone2Name=tail_01, swing1LimitDegrees=5, swing2LimitDegrees=5, twistLimitDegrees=1}
 {bone1Name=tail_03, bone2Name=tail_02, swing1LimitDegrees=10, swing2LimitDegrees=10, twistLimitDegrees=1}
 {bone1Name=tail_04, bone2Name=tail_03, swing1LimitDegrees=15, swing2LimitDegrees=15, twistLimitDegrees=1}
 {bone1Name=tail_05, bone2Name=tail_04, swing1LimitDegrees=20, swing2LimitDegrees=20, twistLimitDegrees=1}
 {bone1Name=tail_06, bone2Name=tail_05, swing1LimitDegrees=25, swing2LimitDegrees=25, twistLimitDegrees=1}
 {bone1Name=tail_07, bone2Name=tail_06, swing1LimitDegrees=30, swing2LimitDegrees=30, twistLimitDegrees=1}
 {bone1Name=tail_08, bone2Name=tail_07, swing1LimitDegrees=30, swing2LimitDegrees=30, twistLimitDegrees=1}
 {
  bone1Name=tail_01
  bone2Name=base_backpack_master
  swing1LimitDegrees=45
  swing1Motion=Free
  swing2LimitDegrees=45
  swing2Motion=Free
  twistLimitDegrees=45
  twistMotion=Free
 }
]
