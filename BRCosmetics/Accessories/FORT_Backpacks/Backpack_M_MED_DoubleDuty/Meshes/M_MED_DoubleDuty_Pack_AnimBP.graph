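# Animation Blueprint for the DoubleDuty backpack cosmetic, layering idle and damage animations with secondary physics.
# A state machine plays random idles, switching to a one-shot reaction when health or shields decrease.
# Pack bone offsets respond to skydive state and spine bend, with a blended offset for skydive floating.
# Three chained RigidBody nodes simulate pack, tail/ears and tail using custom gravity, followed by fixed IK bone corrections.
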AnimBlueprint /BRCosmetics/Accessories/FORT_Backpacks/Backpack_M_MED_DoubleDuty/Meshes/M_MED_DoubleDuty_Pack_AnimBP
target_skeleton=&/M_MED_DoubleDuty_Pack_Skeleton
graphs [
 AnimGraph {nodes [
   Input Pose()
   AnimGraphNode_StateMachine_0 = Idle and Reactive() {is=title AnimGraphNode_StateMachine}
   n1 = "Transform (Modify) Bone - Bone: pack"(
    ComponentPose=Local To Component(LocalPose=@AnimGraphNode_StateMachine_0)
    Alpha=Get is on Ground() -> (is on Ground:double)
    Translation="0,3.3,2"
    Rotation="0,0,0"
   ) {
    translationMode=BMM_Additive
    translationSpace=BCS_BoneSpace
    alphaBoolBlend {blendOption=Linear}
    alphaScaleBiasClamp {bInterpResult=true}
    boneToModify {boneName=pack}
    translation {y=3.3, z=2}
   }
   n2 = "Transform (Modify) Bone - Bone: pack"(
    ComponentPose=@n1
    Translation=Get Backpack Translation() -> (Backpack Translation:Vector)
    Rotation=Get Backpack Rotate() -> (Backpack Rotate:Rotator)
   ) {
    rotationMode=BMM_Additive
    rotationSpace=BCS_ParentBoneSpace
    translationMode=BMM_Additive
    translationSpace=BCS_BoneSpace
    alphaBoolBlend {blendOption=Linear}
    boneToModify {boneName=pack}
   }
   n3 = RigidBody!(ComponentPose=@n2, Alpha, OverrideWorldGravity:Vector=Get GravityOverride(), ExternalForce) -> (Pose) {
    bOverrideWorldGravity=true
    overridePhysicsAsset=&/M_MED_DoubleDuty_Pack_Physics
    simulationSpace=BaseBoneSpace
    alphaBoolBlend {blendOption=Linear}
    baseBoneRef {boneName=root}
    componentAppliedLinearAccClamp=<3000 5500 5500>
    componentLinearAccScale=<0.3 0.55 0.55>
    componentLinearVelScale=<0.3 0.55 0.55>
    simSpaceSettings {
     maxAngularAcceleration=1000
     maxAngularVelocity=3
     maxLinearAcceleration=700
     maxLinearVelocity=500
     velocityScaleZ=0
     worldAlpha=0.2
     externalLinearDragV=<0.2 0.2>
    }
   }
   n4 = RigidBody(ComponentPose=@n3, Alpha=0.2, OverrideWorldGravity=Get GravityOverride()) {
    bOverrideWorldGravity=true
    overridePhysicsAsset=&/M_MED_DoubleDuty_TailEars_Physics
    simulationSpace=BaseBoneSpace
    alphaBoolBlend {blendOption=Linear}
    baseBoneRef {boneName=root}
    componentAppliedLinearAccClamp=<3000 4000 5000>
    componentLinearAccScale=<0.3 0.4 0.5>
    componentLinearVelScale=<0.3 0.4 0.5>
    simSpaceSettings {
     maxAngularAcceleration=1000
     maxAngularVelocity=3
     maxLinearAcceleration=700
     maxLinearVelocity=500
     velocityScaleZ=0
     worldAlpha=0.2
     externalLinearDragV=<0.2 0.2>
    }
   }
   n5 = RigidBody(ComponentPose=@n4, Alpha=0.6, OverrideWorldGravity="0,0,-850") {
    bOverrideWorldGravity=true
    overridePhysicsAsset=&/M_MED_DoubleDuty_Tail_Physics
    simulationSpace=BaseBoneSpace
    alphaBoolBlend {blendOption=Linear}
    baseBoneRef {boneName=root}
    componentAppliedLinearAccClamp=<4500 4000 5000>
    componentLinearAccScale=<0.45 0.4 0.5>
    componentLinearVelScale=<0.45 0.4 0.5>
    simSpaceSettings {
     maxAngularAcceleration=1000
     maxAngularVelocity=3
     maxLinearAcceleration=700
     maxLinearVelocity=500
     velocityScaleZ=0
     worldAlpha=0.2
     externalLinearDragV=<0.2 0.2>
    }
   }
   n6 = "Transform (Modify) Bone - Bone: pack"!(ComponentPose=@n5, bAlphaBoolEnabled:bool=Get bIsSkydiveFloating(), Translation="0,-7,4", Rotation="0,0,-25", Scale) -> (Pose) {
    alphaInputType=Bool
    rotationMode=BMM_Additive
    translationMode=BMM_Additive
    translationSpace=BCS_BoneSpace
    alphaBoolBlend {blendInTime=0.4, blendOption=Linear, blendOutTime=0.4}
    alphaScaleBiasClamp {bInterpResult=true}
    boneToModify {boneName=pack}
    rotation {roll=-25}
    translation {y=-7, z=4}
   }
   n7 = "Transform (Modify) Bone - Bone: Back_00_Back_Ik"!(ComponentPose=@n6, bAlphaBoolEnabled:bool=True, Translation="0.6,0.25,0", Rotation="0,0,0") -> (Pose) {
    alphaInputType=Bool
    rotationMode=BMM_Replace
    rotationSpace=BCS_BoneSpace
    translationMode=BMM_Additive
    translationSpace=BCS_BoneSpace
    alphaBoolBlend {blendInTime=0.4, blendOption=Linear, blendOutTime=0.4}
    alphaScaleBiasClamp {bInterpResult=true}
    boneToModify {boneName=Back_00_Back_Ik}
    translation=<0.6 0.25>
   }
   n8 = "Transform (Modify) Bone - Bone: Back_05_Front_Ik"!(ComponentPose=@n7, bAlphaBoolEnabled:bool=True, Translation="0,0,0.55", Rotation="8.5,0,0") -> (Pose) {
    alphaInputType=Bool
    rotationMode=BMM_Additive
    rotationSpace=BCS_BoneSpace
    translationMode=BMM_Replace
    translationSpace=BCS_BoneSpace
    alphaBoolBlend {blendInTime=0.4, blendOption=Linear, blendOutTime=0.4}
    alphaScaleBiasClamp {bInterpResult=true}
    boneToModify {boneName=Back_05_Front_Ik}
    rotation {pitch=8.5}
    translation {z=0.55}
   }
   Output Pose(Result=Component To Local(ComponentPose=@n8)) {layerGroup=None}
   Get GravityOverride()
  ]}
 Idle and Reactive {
  entry=Idle
  path=AnimGraph.AnimGraphNode_StateMachine_0.Idle and Reactive
  ids=id
  nodes [
   AnimStateNode_0 = Idle()
   AnimStateNode_1 = Reactive() {is=title AnimStateNode}
   AnimStateTransitionNode_0 = AnimStateTransitionNode()
   AnimStateTransitionNode_1 = AnimStateTransitionNode() {bAutomaticRuleBasedOnSequencePlayerInState=true}
  ]
  links [
   {
    dst=Reactive
    rule=AnimGraph.AnimGraphNode_StateMachine_0.Idle and Reactive.AnimStateTransitionNode_0.Transition
    src=Idle
    via=AnimStateTransitionNode_0
   }
   {
    dst=Idle
    rule=AnimGraph.AnimGraphNode_StateMachine_0.Idle and Reactive.AnimStateTransitionNode_1.Transition
    src=Reactive
    via=AnimStateTransitionNode_1
   }
  ]
 }
 Idle {
  path=AnimGraph.AnimGraphNode_StateMachine_0.Idle and Reactive.AnimStateNode_0.Idle
  nodes [
   n1 = Random Sequence Player!() -> (Pose:Pose Link Structure) {
    is=title AnimGraphNode_RandomPlayer
    entries [
     {
      chanceToPlay=0.25
      maxLoopCount=0
      maxPlayRate=1
      minLoopCount=0
      minPlayRate=1
      sequence=/BRCosmetics/Accessories/FORT_Backpacks/Backpack_M_MED_DoubleDuty/Animation/DoubleDuty_Backpack_IdleA
      blendIn {blendOption=Linear, blendTime=0.5, customCurve=null}
     }
     {
      chanceToPlay=0.25
      maxLoopCount=0
      maxPlayRate=1
      minLoopCount=0
      minPlayRate=1
      sequence=/BRCosmetics/Accessories/FORT_Backpacks/Backpack_M_MED_DoubleDuty/Animation/DoubleDuty_Backpack_IdleB
      blendIn {blendOption=Linear, blendTime=0.3, customCurve=null}
     }
     {
      chanceToPlay=0.25
      maxLoopCount=0
      maxPlayRate=1
      minLoopCount=0
      minPlayRate=1
      sequence=/BRCosmetics/Accessories/FORT_Backpacks/Backpack_M_MED_DoubleDuty/Animation/DoubleDuty_Backpack_IdleC
      blendIn {blendOption=Linear, blendTime=0.2, customCurve=null}
     }
     {
      chanceToPlay=0.25
      maxLoopCount=0
      maxPlayRate=1
      minLoopCount=0
      minPlayRate=1
      sequence=/BRCosmetics/Accessories/FORT_Backpacks/Backpack_M_MED_DoubleDuty/Animation/DoubleDuty_Backpack_IdleBreak
      blendIn {blendOption=Linear, blendTime=0.3, customCurve=null}
     }
    ]
   }
   Output Animation Pose(Result=@n1)
  ]
 }
 Reactive {
  path=AnimGraph.AnimGraphNode_StateMachine_0.Idle and Reactive.AnimStateNode_1.Reactive
  nodes [
   n1 = Sequence Player 'DoubleDuty_Backpack_Damage'() {bLoopAnimation=false, sequence=/BRCosmetics/Accessories/FORT_Backpacks/Backpack_M_MED_DoubleDuty/Animation/DoubleDuty_Backpack_Damage}
   Output Animation Pose(Result=@n1)
  ]
 }
 Transition {
  path=AnimGraph.AnimGraphNode_StateMachine_0.Idle and Reactive.AnimStateTransitionNode_0.Transition
  nodes [Result(bCanEnterTransition=Get Reactive Animation() -> (Reactive Animation:bool))]
 }
 Transition {path=AnimGraph.AnimGraphNode_StateMachine_0.Idle and Reactive.AnimStateTransitionNode_1.Transition, nodes [Result()]}
 EventGraph {nodes [
   Event BlueprintUpdateAnimation()
   > n1 = Sequence!(execute) -> (then_0, then_1, then_2:Exec)
   > n2 = Gravity Override State Function DEPRECATED(CharacterPartAnimInstance=Self-Reference(), GameTimeInSeconds=GetGameTimeInSeconds(), GravityOverrideParams=Get Gravity Override Params())
   > Set GravityOverride(GravityOverride=@n2)
   Set is on Ground(execute=@n1.then_2, is on Ground:double=SelectFloat(A=0, B=1, bPickA=Get bIsSkyDiving())) -> (Output_Get:double)
   n3 = Sequence!(execute=@n1.then_1) -> (then_0, then_1)
   > n4 = Branch(Condition=Get bIsSkyDiving())
   n5 = GetOwnerSkeletalMeshComponent()
   n6 = BreakRotator(InRot=BreakTransform(InTransform=GetDeltaTransformFromRefPose(self=@n5, BoneName=spine_05, BaseName=spine_01)).Rotation)
   > Set Backpack Rotate(Backpack Rotate:Rotator=MakeRotator(Roll=MapRangeClamped(Value=@n6.Yaw, InRangeA=0, InRangeB=-40, OutRangeB=15))) -> (Output_Get:Rotator)
   Set Backpack Rotate(execute=@n4.else, Backpack Rotate="0, 0, 0")
   n7 = Branch(execute=@n3.then_1, Condition=Get bIsSkyDiving())
   > Set Backpack Translation(Backpack Translation:Vector=MakeVector(
     Y=MapRangeClamped(Value=@n6.Yaw, InRangeA=0, InRangeB=-30, OutRangeB=4.5)
     Z=MapRangeClamped(Value=@n6.Yaw, InRangeA=0, InRangeB=-30, OutRangeB=3.5)
    )) -> (Output_Get:Vector)
   Set Backpack Translation(execute=@n7.else, Backpack Translation="0, 0, 0")
   Event BlueprintPostEvaluateAnimation() {is=title K2Node_Event}
   > Set Reactive Animation(Reactive Animation:bool=false) -> (Output_Get:bool)
   n8 = Event BlueprintProcessAnimInputEvent()
   > Branch(Condition=OR Boolean(
     A=IsNamed(self=@n8.AnimInputEvent, InEventName=aie.PlayerVitality.ShieldsDecreased)
     B=IsNamed(self=@n8.AnimInputEvent, InEventName=aie.PlayerVitality.HealthDecreased)
    ))
   > Set Reactive Animation(Reactive Animation=true)
   > RetriggerableDelay(Duration=0.2)
   > Set Reactive Animation(Reactive Animation=false)
   TryGetPawnOwner()
  ]}
]
variables [
 Backpack Rotate {type=Rotator}
 Backpack Translation {type=Vector}
 Gravity Override Params {type=GravityOverrideParamsStruct}
 GravityOverride {type=Vector}
 is on Ground {type=number}
 Reactive Animation {type=boolean}
]
